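# Unlit, two-sided masked lightning material for Niagara mesh particles and ribbons, part of the Drift Avatar backpack effects.
# Borrowed Parfait pickaxe functions thicken the spline and jitter it through vertex offset, with dynamic parameters driving width, noise and a sliding gradient mask.
# Oscillating brightness and thickening-derived intensity modulate particle color, with an optional smoke-texture edge.
# Opacity combines mask, smoke or noise and depth fade, while a ribbon parameter switches gradients and disables vertex offset.
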Material /Game/Accessories/FORT_Backpacks/Backpack_M_MED_DriftAvatar/FX/M_Drift_Lightning
graph {
 properties {
  bUsedWithNiagaraMeshParticles=true
  bUsedWithNiagaraRibbons=true
  blendMode=BLEND_Masked
  pixelDepthOffsetMode=PDOM_Legacy
  shadingModel=MSM_Unlit
  twoSided=true
  usedShadingModels=Unlit
 }
 nodes [
  n1 = TextureCoordinate()
  ComponentMask(None=@n1) {r=true}
  n2 = LinearInterpolate(
   A=VectorParameter() {parameterName=Default Color, defaultValue=<rgba 260.8502 0 333 3>}.RGB
   B=ParticleColor().RGB
   Alpha=ScalarParameter() {defaultValue=1, group=Niagara?, parameterName=Use Particle}
  )
  n3 = LinearInterpolate(
   B=MaterialFunctionCall!(In) {materialFunction=/Game/Weapons/FORT_Melee/Pickaxe_Parfait_Female/FX/Materials/Functions/MF_TrifleParfait_BrightnessOscillation}.Brightness
   Alpha=ScalarParameter() {defaultValue=1, parameterName=Use Brightness occilation}
  ) {constA=1}
  n4 = Power(Base=@n3, Exp=ScalarParameter() {defaultValue=3, parameterName=Brightness Oscillation Exponent})
  n5 = TextureCoordinate()
  n6 = VectorParameter() {parameterName=Width Scale Options, channelNames {a=Speed Y, b=Speed X, g=Tile Y, r=Tile X}, defaultValue=<rgba 1 1 0.1 1>}
  n7 = AppendVector(A=@n6.Tile X, B=@n6.Tile X)
  n8 = TextureSampleParameter2D(UVs=Panner(
    Coordinate=Multiply(A=@n5, B=@n7)
    Time=DynamicParameter() {parameterIndex=1, paramNames [Width Scale Time, NA, NA, NA]}.Width Scale Time
    Speed=AppendVector(A=@n6.Speed X, B=@n6.Speed X)
   )) {parameterName=Width Scale, texture=/Engine/EditorLandscapeResources/WhiteSquareTexture}
  n9 = Max(B=LinearInterpolate(
    A=ScalarParameter() {parameterName=Width Scale Low}
    B=ScalarParameter() {defaultValue=1, parameterName=Width Scale High}
    Alpha=ChannelMaskParameter(None=@n8.RGBA) {parameterName=WidthScale Channel}.R
   ))
  n10 = ShadingPathSwitch(Default=QualitySwitch(Default=@n9, Low=@n9, High=@n9, Medium=@n9, Epic=@n9), Mobile=QualitySwitch(Default=@n9, Low=Constant() {r=1}))
  n11 = DynamicParameter() {paramNames [Width, Mask Size, Mask Position, Noise Intensity]}
  n12 = MaterialFunctionCall!(
   Noise Intensity=@n11.Noise Intensity
   Noise UV Scale=ScalarParameter() {defaultValue=0.2, parameterName=Noise UV Scale}
   Time
   Noise Axis Rotation
  ) {materialFunction=/Game/Weapons/FORT_Melee/Pickaxe_Parfait_Female/FX/Materials/Functions/MF_TrifleParfait_Noise}
  n13 = Multiply(A=@n3, B=@n12.Noise Width Multiplier)
  n14 = TextureCoordinate()
  n15 = VectorParameter() {group=Alternative Noise, parameterName=Alter Spline Noise Tex, sortPriority=1, defaultValue=<rgba 1 1 0 0>}
  n16 = AppendVector(A=@n15.R, B=@n15.R)
  n17 = TextureSampleParameter2D(UVs=Panner(Coordinate=Multiply(A=@n14, B=@n16), Speed=AppendVector(A=@n15.B, B=@n15.B)) {speedX=0.1}) {
   group=Alternative Noise
   parameterName=Spline Noise
   samplerType=SAMPLERTYPE_LinearGrayscale
   texture=/Game/Effects/Fort_Effects/Emotes/UltraLight/Texture/T_Plasma_C_Greyscale
  }
  n18 = Multiply(A=@n12.Noise Width Multiplier, B=Clamp(None=@n17.RGB))
  n19 = ScalarParameter() {group=Alternative Noise, parameterName=Alter Spline Noise}
  n20 = Constant() {r=1.333}
  n21 = ScalarParameter() {defaultValue=1, group=Niagara?, parameterName=Use Particle}
  n22 = ComponentMask(None=@n1) {g=true}
  n23 = OneMinus(None=@n22)
  n24 = Round(None=ParticleRandom())
  n25 = Time()
  n26 = MaterialFunctionCall!(
   Gradient=LinearInterpolate(A=@n22, B=LinearInterpolate(A=@n22, B=@n23, Alpha=@n24), Alpha=ScalarParameter() {defaultValue=1, parameterName=UseParticleRandomFlip})
   Mask Position=LinearInterpolate(A=Frac(None=@n25), B=@n11.Mask Position, Alpha=@n21) {constA=0.5}
   Mask Size=LinearInterpolate(A=Constant() {r=3}, B=@n11.Mask Size, Alpha=@n21)
  ) {materialFunction=/Game/Effects/Fort_Effects/MaterialFunctions/MF_GradientMask}
  n27 = Multiply(
   A=Multiply(A=LinearInterpolate(A=@n13, B=@n18, Alpha=@n19) {constA=1, constAlpha=1}, B=LinearInterpolate(A=@n20, B=@n11.Width, Alpha=@n21))
   B=@n26.Sinusoidal
  )
  n28 = MaterialFunctionCall!(Width=StaticSwitchParameter(True=Multiply(A=@n10, B=@n27), False=@n27) {parameterName=UseTextureToScaleWidth}) {materialFunction=/Game/Weapons/FORT_Melee/Pickaxe_Parfait_Female/FX/Materials/Functions/MF_TrifleParfait_SplineThicken}
  n29 = Power(Base=@n28.Distance Intensity Modulation, Exp=ScalarParameter() {defaultValue=1.25, parameterName=Distance Intensity Modulation Exponent})
  n30 = Max(A=VectorParameter() {parameterName=Minimum Color, defaultValue=<rgba 0.3625 0.15 1 0>}.RGB, B=Multiply(A=Multiply(A=@n2, B=@n4), B=@n29))
  n31 = Multiply(A=ParticleColor().RGB, B=@n4)
  n32 = Multiply(A=@n31, B=@n29)
  n33 = ScalarParameter() {defaultValue=1, group=Niagara?, parameterName=Use Particle}
  n34 = LinearInterpolate(A=VertexInterpolator(VS=LinearInterpolate(A=@n30, B=@n32, Alpha=@n33)).PS, B=@n31, Alpha=ScalarParameter() {parameterName=Use With Ribbon}) {constB=0}
  n35 = TextureCoordinate()
  n36 = VectorParameter() {group=Opacity, parameterName=SmokeTrail UV, sortPriority=1, defaultValue=<rgba 1 1 0 0>}
  n37 = AppendVector(A=@n36.R, B=@n36.R)
  n38 = LinearInterpolate(
   A=TextureCoordinate()
   B=MaterialFunctionCall(XY=Multiply(A=@n35, B=@n37)) {materialFunction=&Swizzle}.YX
   Alpha=ScalarParameter() {defaultValue=1, parameterName=Swizzle}
  )
  n39 = TextureSampleParameter2D(UVs=Panner(Coordinate=@n38, Speed=AppendVector(A=@n36.B, B=@n36.B))) {group=Opacity, parameterName=SmokeTrailTexture, texture=/Game/Effects/Fort_Effects/Textures/Smoke/T_Smoke_Trail_Soft_Chromatic}
  n40 = Multiply(A=SmoothStep(Value=@n39.R) {constMax=0, constMin=0.5}, B=VectorParameter() {parameterName=EdgeColor, defaultValue=<rgba 0 0 0 1>}.RGB) {constB=5}
  MaterialFunctionCall(Emissive=LinearInterpolate(A=@n34, B=Add(A=@n34, B=@n40), Alpha=ScalarParameter() {parameterName=Add Edge})) {materialFunction=&MF_FX_EmissiveAdjustments}
  n41 = ParticleColor()
  n42 = Multiply(A=@n41.A, B=Multiply(A=@n26.Sinusoidal, B=@n39.R))
  n43 = OneMinus(None=ComponentMask(None=@n38) {r=true})
  n44 = ScalarParameter() {parameterName=Use Ribbon Gradient}
  n45 = Multiply(A=@n41.A, B=Multiply(A=LinearInterpolate(B=@n43, Alpha=@n44) {constA=1}, B=@n39.R))
  n46 = ScalarParameter() {parameterName=Use With Ribbon}
  n47 = Multiply(
   A=MaterialFunctionCall!(Noise Intensity=VertexInterpolator(VS=@n29).PS) {materialFunction=/Game/Characters/Player/Male/Medium/Bodies/M_MED_Ebony_Spark/FX/MF_EbonySpark_OpacityNoise}.Masked Opacity
   B=@n26.Sinusoidal
  )
  Multiply(
   A=LinearInterpolate(A=LinearInterpolate(A=@n42, B=@n45, Alpha=@n46) {constB=0}, B=@n47, Alpha=ScalarParameter() {parameterName=Tanslucent Masked})
   B=DepthFade(FadeDistance=ScalarParameter() {defaultValue=25, parameterName=Depth Fade})
  )
  n48 = DynamicParameter() {parameterIndex=3, paramNames [X,Y,Z,Param4]}
  n49 = AppendVector(A=@n48.X, B=@n48.X)
  n50 = OneMinus(None=ComponentMask(None=TextureCoordinate()) {g=true})
  NamedRerouteDeclaration(None=Multiply(A=AppendVector(A=@n49, B=@n48.Z), B=Power(Base=@n50) {constExponent=1.5})) {name=WPOOffset, nodeColor=<rgba 0.5469 1 0 1>}
  n51 = TextureCoordinate()
  n52 = VectorParameter() {group=Alternative Noise, parameterName=Alter WPO Noise Tex, sortPriority=1, defaultValue=<rgba 1 1 0 0>}
  n53 = AppendVector(A=@n52.R, B=@n52.R)
  n54 = Multiply(A=@n52.A, B=LinearInterpolate(Alpha=VertexColor()) {constA=1, constB=1.1})
  n55 = TextureSampleParameter2D(UVs=Panner(Coordinate=Multiply(A=@n51, B=@n53), Speed=AppendVector(A=@n52.B, B=@n54)) {speedX=0.1}) {
   group=Alternative Noise
   parameterName=WPO Noise Alter
   samplerType=SAMPLERTYPE_LinearGrayscale
   texture=&T_Plasma_C_Greyscale
  }
  n56 = Subtract(A=Clamp(None=@n55.RGB)) {constB=0.5}
  n57 = Add(
   A=LinearInterpolate(
    A=@n12.Noise WPO
    B=Multiply(A=@n56, B=@n11.Noise Intensity)
    Alpha=ScalarParameter() {group=Alternative Noise, parameterName=Alter WPO Noise}
   ) {constA=1, constAlpha=0}
   B=@n28.WPO
  )
  n58 = Add(A=@n57, B=NamedRerouteUsage())
  n59 = ScalarParameter() {parameterName=Use WPO Offset}
  NamedRerouteDeclaration(None=LinearInterpolate(A=LinearInterpolate(A=@n57, B=@n58, Alpha=@n59) {constB=0}, Alpha=ScalarParameter() {parameterName=Use With Ribbon}) {constB=0}) {name=WPO, nodeColor=<rgba 0 1 0.1562 1>}
  NamedRerouteUsage()
  n60 = ParticleColor()
  n61 = ScalarParameter() {defaultValue=1, group=Niagara?, parameterName=Use Particle}
  Saturate(None=Multiply(A=@n47, B=LinearInterpolate(B=@n60.A, Alpha=@n61) {constA=1}))
 ]
}
